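; BP_Labs_DangerRoomWall_A_BSG is a Labs set Danger Room wall Blueprint that exposes its projector pucks as simple destructible props.
; It supplies four ProjectorPuck components (03 to 06) as destructible sources and gives every one the same fixed entry, with Wood material values and no attachment.
; It flags the source actor to be kept.
; BeginPlay, overlap, Tick and the construction script contain no logic.
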(event EventBeginPlay)

(event Collision|EventActorBeginOverlap
  (params (OtherActor :type Actor%20Object%20Reference)))

(event EventTick
  (params (DeltaSeconds :type Float%20%28single-precision%29)))

(fn GetSimpleDestructiblePropComponents
  (return (Utilities|Array|MakeArray (Variables|Default|GetProjectorPuck_04) (Variables|Default|GetProjectorPuck_05) (Variables|Default|GetProjectorPuck_06) (Variables|Default|GetProjectorPuck_03))))

(fn GetSimpleDestructiblePropEntry
  (params (SourceIndex :type Integer) (SourceComponent :type Static%20Mesh%20Component%20Object%20Reference))
  (return 0 0 "(Name=\"\")" 0 "(CurveTable=None,RowName=\"\")" "Wood" "Wood" "ATTACH_None" "SLOT_None" false false false))

(fn ShouldKeepSourceActor
  (return true))

(fn ConstructionScript)
