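; Curved wall Blueprint for the Labs Danger Room set that exposes its projector pucks as simple destructible props.
; GetSimpleDestructiblePropComponents returns eight ProjectorPuck components, and ShouldKeepSourceActor returns true so the wall actor is kept.
; GetSimpleDestructiblePropEntry ignores its inputs and returns one fixed, mostly empty entry with Wood types and no attachment or slot.
; The event graph holds only disabled stubs, and the construction script just calls its parent.
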(event EventBeginPlay
  (|CallParentFunction) ; This node is disabled and will not be called.\nDrag off pins to build functionality.
)

(event Collision|EventActorBeginOverlap
  (params (OtherActor :type Actor%20Object%20Reference))
  (|CallParentFunction OtherActor) ; This node is disabled and will not be called.\nDrag off pins to build functionality.
)

(event EventTick
  (params (DeltaSeconds :type Float%20%28single-precision%29))
  (|CallParentFunction DeltaSeconds) ; This node is disabled and will not be called.\nDrag off pins to build functionality.
)

(fn GetSimpleDestructiblePropComponents
  (return (Utilities|Array|MakeArray (Variables|Default|GetProjectorPuck1) (Variables|Default|GetProjectorPuck2) (Variables|Default|GetProjectorPuck3) (Variables|Default|GetProjectorPuck) (Variables|Default|GetProjectorPuck6) (Variables|Default|GetProjectorPuck5) (Variables|Default|GetProjectorPuck4) (Variables|Default|GetProjectorPuck7))))

(fn GetSimpleDestructiblePropEntry
  (params (SourceIndex :type Integer) (SourceComponent :type Static%20Mesh%20Component%20Object%20Reference))
  (return 0 0 "(Name=\"\")" 0 "(CurveTable=None,RowName=\"\")" "Wood" "Wood" "ATTACH_None" "SLOT_None" false false false))

(fn ShouldKeepSourceActor
  (return true))

(fn ConstructionScript
  (|CallParentFunction))
