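; BP_Labs_Shelf_Str is a Labs set wall shelf Blueprint, with logic that declares which assets and classes it exposes through Fort's exposed asset provider.
; Its overrides call the parent, then add provider version 1 and collect asset references from the MaterialInstance property.
; It also registers indirect class references for its GraphInstance, ISMComponent, LandscapeCache and WorldActor class variables.
; The construction script is empty.
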(fn BuildExposedAssetProviderVersion(Implementation)
  (params (Version :type Fort%20Exposed%20Asset%20Provider%20Version%20Structure%20%28by%20ref%29))
  (|CallParentFunction Version)
  (ExposedAssetProvider|AddVersion Version 1)
  (return))

(fn CollectAssetReferencestoAllow(Implementation)
  (params (Collector :type Fort%20Exposed%20Asset%20Reference%20Collector%20Structure%20%28by%20ref%29))
  (|CallParentFunction Collector)
  (ExposedAssetProvider|CollectObjectPropertyAssetReferences Collector 0 (Variables|Default|GetMaterialInstance))
  (return))

(fn CollectClassReferencestoAllow(Implementation)
  (params (Collector :type Fort%20Exposed%20Class%20Reference%20Collector%20Structure%20%28by%20ref%29))
  (|CallParentFunction Collector)
  (ExposedAssetProvider|CollectIndirectClassReference Collector (Variables|Default|GetGraphInstanceClass))
  (ExposedAssetProvider|CollectIndirectClassReference Collector (Variables|Default|GetISMComponentClass))
  (ExposedAssetProvider|CollectIndirectClassReference Collector (Variables|Default|GetLandscapeCacheClass))
  (ExposedAssetProvider|CollectIndirectClassReference Collector (Variables|Default|GetWorldActorClass))
  (return))

(fn ConstructionScript)
